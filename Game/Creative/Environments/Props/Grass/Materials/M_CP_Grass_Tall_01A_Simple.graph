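# Masked two-sided foliage material for the tall grass prop in Creative, set up for instanced static meshes.
# Mask texture channels drive opacity, wind falloff and subsurface; one specular-texture channel feeds metallic, specular and roughness alike.
# The normal map is transformed and flipped by TwoSidedSign, and a gameplay grass-wind function supplies world position offset with a sink-into-ground distance.
# Attributes pass through the tree subsurface function, then MF_CreativeMoveTool.
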Material /Game/Creative/Environments/Props/Grass/Materials/M_CP_Grass_Tall_01A_Simple
graph {
 properties {
  bCastRayTracedShadows=false
  bUseMaterialAttributes=true
  bUsedWithInstancedStaticMeshes=true
  blendMode=BLEND_Masked
  pixelDepthOffsetMode=PDOM_Legacy
  shadingModel=MSM_TwoSidedFoliage
  twoSided=true
  usedShadingModels=Two Sided Foliage
 }
 nodes [
  n1 = TextureSample() {texture=/Game/Creative/Environments/Props/Grass/Textures/T_CP_Grass_Tall_01A_S}
  n2 = TextureSample() {samplerType=SAMPLERTYPE_Masks, texture=/Game/Creative/Environments/Props/Grass/Textures/T_CP_Grass_Tall_01A_M}
  n3 = Transform(None=ComponentMask(None=TextureSample() {samplerType=SAMPLERTYPE_Normal, texture=/Game/Creative/Environments/Props/Grass/Textures/T_CP_Grass_Tall_01A_N}.RGB) {b=true, desc=Normal, g=true, r=true})
  n4 = TwoSidedSign()
  n5 = MaterialFunctionCall!(
   SinkIntoGroundDistance=ScalarParameter() {defaultValue=300, group=WPO-DistanceBased, parameterName=SinkIntoGroundDistance}
   WindFalloffMask=@n2.G
   AfterHeightModification=@n5.Wind
  ) {materialFunction=/Game/Athena/Apollo/Environments/Landscape/SmallFoliage/MasterMaterial/MaterialFunctions/MF_GrassWind-GameplayWPO}
  n6 = MaterialFunctionCall!(
   In=MakeMaterialAttributes!(
    BaseColor=TextureSample() {texture=/Game/Creative/Environments/Props/Grass/Textures/T_CP_Grass_Tall_01A_BC}.RGB
    Metallic=@n1.G
    Specular=@n1.G
    Roughness=@n1.G
    Anisotropy
    EmissiveColor
    Opacity
    OpacityMask=@n2.R
    Normal=Multiply(A=@n3, B=@n4)
    Tangent
    WorldPositionOffset=@n5.FinalWPO
    SubsurfaceColor
    ClearCoat
    ClearCoatRoughness
    AmbientOcclusion
    Refraction
    CustomizedUVs_0
    CustomizedUVs_1
    CustomizedUVs_2
    CustomizedUVs_3
    CustomizedUVs_4
    CustomizedUVs_5
    CustomizedUVs_6
    CustomizedUVs_7
    PixelDepthOffset
    ShadingModel
    Displacement
   )
   Mask=@n2.B
  ) {
   desc="Uses Red Channel of Mask Texture "
   materialFunction=/Game/Packages/Fortnite_UniqueMaterials_1/MaterialFunctions/Trees/MF_Tree_SubSurface
  }
  MaterialFunctionCall(MA In=@n6.Result) {materialFunction=&MF_CreativeMoveTool}
 ]
}
